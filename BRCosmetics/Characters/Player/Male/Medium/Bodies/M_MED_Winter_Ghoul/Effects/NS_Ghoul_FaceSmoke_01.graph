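# Niagara face smoke effect for the Winter Ghoul player cosmetic, built from mesh particles.
# The active Particle_Emitter burst-spawns an offset mesh rotated 90 degrees, tints it white with boosted alpha, and keeps it alive past its lifetime.
# A disabled Smokey emitter holds a movement-based trail: distance-spawned, randomly oriented meshes with drag, curve-driven color scaling and dynamic material parameters animated over lifetime.
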NiagaraSystem /BRCosmetics/Characters/Player/Male/Medium/Bodies/M_MED_Winter_Ghoul/Effects/NS_Ghoul_FaceSmoke_01
dependencies {
 usedDataInterfaces [&NiagaraDataInterfaceVectorCurve, &NiagaraDataInterfaceCurve]
 usedDynamicInputs []
 usedModules [&SystemState]
}
emitters {
 Particle_Emitter {
  emitterUpdateScript [
   &EmitterState(
    Inactive Response=::NewEnumerator2
    Life Cycle Mode=::NewEnumerator1
    Loop Behavior=::NewEnumerator1
    Loop Duration Mode=::NewEnumerator1
    UseLoopDelay=true
   )
   &SpawnBurst_Instantaneous(Use Spawn Probability=false, UseLoopCountLimit=false)
  ]
  particleSpawnScript [
   /Niagara/Modules/Spawn/Initialization/InitializeParticle(
    Mesh Scale=<0 1.875 0>
    Module.Write Color=true
    Module.Write Lifetime=true
    Module.Write Mass=true
    Module.Write Material Random=true
    Module.Write Position=true
    Module.Write Scale=true
    Module.Write SpriteRotation=false
    Module.Write SpriteSize=false
   ) {enabled=false}
   /Niagara/Modules/Spawn/Initialization/V2/InitializeParticle(Mesh Scale Mode=::NewEnumerator3, Position Offset=<4 6 0>, UsePositionOffset=true, Write Material Random=true) {module=InitializeParticle001}
   /SetVariables(Particles.MeshOrientation=&MakeQuatFromAxisAngle(Axis=<0 1 0>, AngleInDegrees=90)) {isSetParameters=true}
  ]
  particleUpdateScript [
   &ParticleState(Kill Particles When Lifetime Has Elapsed=false, Let Infinitely Lived Particles Die When Emitter Deactivates=true)
   &Color(Color=&MakeLinearColorFromVectorAndFloat("Vector (RGB)"=<1 1 1>, "Float (Alpha)"=3))
  ]
  rendererClasses [&NiagaraMeshRendererProperties]
  renderers [{rendererClass=&NiagaraMeshRendererProperties, rendererIndex=0}]
 }
 Smokey {
  bEnabled=false
  emitterUpdateScript [
   &EmitterState(Life Cycle Mode=::NewEnumerator1, Loop Duration=1, UseLoopDelay=true)
   &SpawnPerUnit(Module.Use Max Movement Threshold=true, Module.Use Movement Tolerance=true, Movement Tolerance=0.1, Use Spawn Probability=false) {enabled=false}
   &SpawnPerUnit(Module.Use Max Movement Threshold=true, Module.Use Movement Tolerance=true, Movement Tolerance=0.1, Use Spawn Probability=false) {module=SpawnPerUnit001}
  ]
  particleSpawnScript [
   /SetVariables(Particles.MeshOrientation=&MakeQuatFromAxisAngle(Axis=&RandomRangeVector(Minimum=<0 0 0>, Maximum=<0 1.875 0>), AngleInDegrees=90)) {isSetParameters=true}
   /Niagara/Modules/Spawn/Initialization/InitializeParticle(
    Lifetime=&RandomRangeFloat(Minimum=1.5, Maximum=1)
    Mesh Scale=&RandomRangeVector(Minimum=<0 1.575 2>, Maximum=<0 1.7 0>)
    Module.Write Color=true
    Module.Write Lifetime=true
    Module.Write Mass=true
    Module.Write Material Random=true
    Module.Write Position=true
    Module.Write Scale=true
    Module.Write SpriteRotation=false
    Module.Write SpriteSize=false
   ) {enabled=false}
   /Niagara/Modules/Spawn/Initialization/V2/InitializeParticle(Lifetime Max=1.5, Lifetime Mode=::NewEnumerator1, Mesh Scale Mode=::NewEnumerator3, UsePositionOffset=false, Write Material Random=true) {module=InitializeParticle001}
   &Color(Color=&MakeLinearColorFromVectorAndFloat("Vector (RGB)"=<0 0 0>, "Float (Alpha)"=&ClampFloat(Float=30)))
   &DynamicMaterialParameters(
    Index 0 Param 2=&RandomRangeFloat(Minimum=30, Maximum=50)
    Module.Param0WriteEnabled=false
    Module.Param1WriteEnabled=true
    Module.Param2WriteEnabled=false
    Module.Param3WriteEnabled=false
   ) {module=DynamicMaterialParameters001}
  ]
  particleUpdateScript [
   &ParticleState()
   &ScaleColor(
    Scale Alpha=&FloatFromCurve(FloatCurve=DataInterface() {propertyValues {Curve {keys [
         {arriveTangent=0.2972, interpMode=RCIM_Cubic, time=0, value=-0.0092}
         {arriveTangent=0.0092, interpMode=RCIM_Cubic, leaveTangent=0.0092, time=0.271, value=1.5228}
         <at 1 0 RCIM_Cubic>
        ]}}})
    Scale RGB=&VectorFromCurve(VectorCurve=DataInterface() {propertyValues {
       XCurve {keys [<at 0.0015 1 RCIM_Cubic>,<at 0.5 1 RCIM_Cubic>,<at 0.5619 1 RCIM_Cubic>,<at 1 1 RCIM_Cubic>]}
       YCurve {keys [
         <at 0.0015 22.0935 RCIM_Cubic>
         {arriveTangent=-39.1079, interpMode=RCIM_Cubic, leaveTangent=-39.1079, time=0.5, value=5}
         {arriveTangent=-10, interpMode=RCIM_Cubic, leaveTangent=-10, time=0.5619, value=0.1782}
         <at 1 0 RCIM_Cubic>
        ]}
       ZCurve {keys [
         <at 0.0015 0 RCIM_Cubic>
         {arriveTangent=8.9225, interpMode=RCIM_Cubic, leaveTangent=8.9225, time=0.5, value=5}
         {arriveTangent=490, interpMode=RCIM_Cubic, leaveTangent=490, time=0.5619, value=5}
         <at 1 250 RCIM_Cubic>
        ]}
      }})
    ScaleA=true
    ScaleRGB=true
   )
   &DynamicMaterialParameters(
    Index 0 Param 1=&FloatFromCurve(FloatCurve=DataInterface() {propertyValues {Curve {keys [<at 0 5>,<at 1 0>]}}})
    Index 0 Param 3=&FloatFromCurve(FloatCurve=DataInterface() {propertyValues {Curve {keys [<at 0 3>,<at 1 12>]}}})
    Index 0 Param 4=&FloatFromCurve(FloatCurve=DataInterface() {propertyValues {Curve {keys [
         {arriveTangent=-0.0169, interpMode=RCIM_Cubic, time=0.0123, value=0.25}
         {arriveTangent=-1.0347, interpMode=RCIM_Cubic, leaveTangent=-1.0347, time=0.8212, value=0.2237}
         {arriveTangent=-6.8435, interpMode=RCIM_Cubic, leaveTangent=-6.8435, time=0.9123, value=-0.6812}
         <at 1 -1 RCIM_Cubic>
        ]}}})
    Module.Param0WriteEnabled=true
    Module.Param1WriteEnabled=false
    Module.Param2WriteEnabled=true
    Module.Param3WriteEnabled=true
   )
   &ScaleMeshSize(Scale Factor=&VectorFromCurve(VectorCurve=DataInterface() {propertyValues {XCurve {keys [<at 0 1>,<at 1 1>]}, YCurve {keys [<at 0 1>,<at 1 1>]}, ZCurve {keys [<at 0 1>,<at 1 1>]}}}))
   &Drag(Use Linear Drag=true, Use Rotational Drag=false)
   &SolveForcesAndVelocity(Clamp Velocity=false, Limit Acceleration=false)
  ]
  rendererClasses [&NiagaraMeshRendererProperties]
  renderers [{rendererClass=&NiagaraMeshRendererProperties, rendererIndex=0}]
 }
}
summary {
 systemName=NS_Ghoul_FaceSmoke_01
 emitters [
  {emitterName=Particle_Emitter, rendererClasses [&NiagaraMeshRendererProperties]}
  {bEnabled=false, emitterName=Smokey, rendererClasses [&NiagaraMeshRendererProperties]}
 ]
 userVariables []
}
